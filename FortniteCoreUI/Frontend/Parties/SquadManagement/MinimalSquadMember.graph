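# Minimal squad member widget for the party UI, showing one member as a tinted meeple icon with a check icon and a local-player arrow.
# Member establishment sets empty or not-ready visuals and shows the arrow only for the local member.
# Readiness changes restyle the meeple through material outline and tint functions, while Ready plays the OnReady flash-and-pop animation.
# The Style animations duplicate those functions and are effectively unused; SetReadyVisuals is empty.
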WidgetBlueprint /FortniteCoreUI/Frontend/Parties/SquadManagement/MinimalSquadMember
animations [
 OnReady {
  binding_count=3
  length_seconds=0.3
  object=&self:OnReady
  sequence {bindings [
    ReadyFlash {
     id=#1
     possessed_class=Image
     tracks [
      {
       type=MovieSceneByteTrack
       display_name=Visibility
       sections [{type=MovieSceneByteSection, channels [None {type=MovieSceneScriptingByteChannel, keys [[0,4],[6,1]]}]}]
      }
      {
       type=MovieSceneFloatTrack
       display_name=Render Opacity
       sections [{type=MovieSceneFloatSection, channels [None {type=MovieSceneScriptingFloatChannel, keys [[0,0],[1,1],[6.0003,0]]}]}]
      }
      {
       type=MovieScene2DTransformTrack
       sections [{
         type=MovieScene2DTransformSection
         channels [
          Translation.X {type=MovieSceneScriptingFloatChannel}
          Translation.Y {type=MovieSceneScriptingFloatChannel}
          Angle {type=MovieSceneScriptingFloatChannel}
          Scale.X {type=MovieSceneScriptingFloatChannel, keys [[1,1],[6,3],[6.0003,1.5]]}
          Scale.Y {type=MovieSceneScriptingFloatChannel, keys [[1,1],[6,3],[6.0003,1.5]]}
          Shear.X {type=MovieSceneScriptingFloatChannel}
          Shear.Y {type=MovieSceneScriptingFloatChannel}
         ]
        }]
      }
     ]
    }
    Image_ReadyIcon {
     id=#2
     possessed_class=Image
     tracks [{
       type=MovieScene2DTransformTrack
       sections [{
         type=MovieScene2DTransformSection
         channels [
          Translation.X {type=MovieSceneScriptingFloatChannel}
          Translation.Y {type=MovieSceneScriptingFloatChannel, keys [[0,2],[2,-4],[4,-4],[6,2]]}
          Angle {type=MovieSceneScriptingFloatChannel}
          Scale.X {type=MovieSceneScriptingFloatChannel, keys [[0,1],[2,1.5],[4,1.5],[6,1]]}
          Scale.Y {type=MovieSceneScriptingFloatChannel, keys [[0,1],[2,1.5],[4,1.5],[6,1]]}
          Shear.X {type=MovieSceneScriptingFloatChannel}
          Shear.Y {type=MovieSceneScriptingFloatChannel}
         ]
        }]
      }]
    }
    MeepleContainer {
     id=#3
     possessed_class=Overlay
     tracks [{
       type=MovieScene2DTransformTrack
       sections [{
         type=MovieScene2DTransformSection
         channels [
          Translation.X {type=MovieSceneScriptingFloatChannel}
          Translation.Y {type=MovieSceneScriptingFloatChannel, keys [[0,0],[2,-4],[4,-4],[6,0]]}
          Angle {type=MovieSceneScriptingFloatChannel}
          Scale.X {type=MovieSceneScriptingFloatChannel, keys [[0,1],[2,1.25],[4,1.25],[6,1]]}
          Scale.Y {type=MovieSceneScriptingFloatChannel, keys [[0,1],[2,1.25],[4,1.25],[6,1]]}
          Shear.X {type=MovieSceneScriptingFloatChannel}
          Shear.Y {type=MovieSceneScriptingFloatChannel}
         ]
        }]
      }]
    }
   ]}
 }
 StyleEmpty {
  binding_count=1
  length_seconds=0.2
  object=&self:StyleEmpty
  sequence {bindings [Image_Meeple {
     id=#4
     possessed_class=Image
     tracks [
      {
       type=MovieSceneWidgetMaterialTrack
       sections [{type=MovieSceneParameterSection, channels [UsesOutline {type=MovieSceneScriptingFloatChannel, const [1,2]}]}]
      }
      {
       type=MovieSceneColorTrack
       display_name=Color and Opacity
       sections [{
         type=MovieSceneColorSection
         channels [
          Color.R {type=MovieSceneScriptingFloatChannel, const [0.2542,2]}
          Color.G {type=MovieSceneScriptingFloatChannel, const [0.4969,2]}
          Color.B {type=MovieSceneScriptingFloatChannel, const [0.7913,2]}
          Color.A {type=MovieSceneScriptingFloatChannel, const [1,2]}
         ]
        }]
      }
     ]
    }]}
 }
 StyleSitOut {
  binding_count=1
  length_seconds=0.2
  object=&self:StyleSitOut
  sequence {bindings [Image_Meeple {
     id=#4
     possessed_class=Image
     tracks [
      {
       type=MovieSceneWidgetMaterialTrack
       sections [{type=MovieSceneParameterSection, channels [UsesOutline {type=MovieSceneScriptingFloatChannel, const [0,2]}]}]
      }
      {
       type=MovieSceneColorTrack
       display_name=Color and Opacity
       sections [{
         type=MovieSceneColorSection
         channels [
          Color.R {type=MovieSceneScriptingFloatChannel, const [0.2542,2]}
          Color.G {type=MovieSceneScriptingFloatChannel, const [0.4969,2]}
          Color.B {type=MovieSceneScriptingFloatChannel, const [0.7913,2]}
          Color.A {type=MovieSceneScriptingFloatChannel, const [1,2]}
         ]
        }]
      }
     ]
    }]}
 }
 StyleNotReady {
  binding_count=1
  length_seconds=0.2
  object=&self:StyleNotReady
  sequence {bindings [Image_Meeple {
     id=#4
     possessed_class=Image
     tracks [
      {
       type=MovieSceneWidgetMaterialTrack
       sections [{type=MovieSceneParameterSection, channels [UsesOutline {type=MovieSceneScriptingFloatChannel, const [0,2]}]}]
      }
      {
       type=MovieSceneColorTrack
       display_name=Color and Opacity
       sections [{
         type=MovieSceneColorSection
         channels [
          Color.R {type=MovieSceneScriptingFloatChannel, const [0.4452,2]}
          Color.G {type=MovieSceneScriptingFloatChannel, const [1,2]}
          Color.B {type=MovieSceneScriptingFloatChannel, const [1,2]}
          Color.A {type=MovieSceneScriptingFloatChannel, const [1,2]}
         ]
        }]
      }
     ]
    }]}
 }
]
graphs [
 EventGraph {nodes [
   n1 = PlayAnimationForward(InAnimation=Get StyleNotReady() -> (StyleNotReady:Widget Animation Object Reference))
   n2 = Branch(execute=[@n1.then, @n9.then], Condition=IsCurrentMemberLocal!(self:Fort Team Member Entry Base Object Reference) -> (ReturnValue:bool))
   n3 = Get LocalUserIndicator() -> (LocalUserIndicator:Image Object Reference)
   > SetVisibility(self=@n3, InVisibility=SelfHitTestInvisible)
   SetVisibility(execute=[@n11.then, @n10.then, @n2.else], self=@n3, InVisibility=Collapsed)
   n4 = Event OnMemberGameReadinessChanged!() -> (OutputDelegate, then, ReadyStatus:EGameReadiness Enum=NotReady) {is=title K2Node_Event}
   > n5 = Branch(Condition="Equal (Enum)"(A:EGameReadiness Enum=@n4.ReadyStatus, B:EGameReadiness Enum=NotReady))
   > SetNotReadyVisuals()
   n6 = Branch(execute=@n5.else, Condition="Equal (Enum)"(A=@n4.ReadyStatus, B=Ready))
   > PlayAnimationForward(InAnimation=Select!(
     NotReady:Widget Animation Object Reference=Get StyleNotReady()
     Ready:Widget Animation Object Reference=Get OnReady() -> (OnReady:Widget Animation Object Reference)
     Reevaluating:Widget Animation Object Reference
     SittingOut:Widget Animation Object Reference=Get StyleSitOut() -> (StyleSitOut:Widget Animation Object Reference)
     Away:Widget Animation Object Reference
     Index:EGameReadiness Enum=@n4.ReadyStatus
    ) -> (ReturnValue:Widget Animation Object Reference))
   Branch(execute=@n6.else, Condition="Equal (Enum)"(A=@n4.ReadyStatus, B=SittingOut))
   > SetSitOutVisuals()
   n7 = Event BP_OnTeamMemberEstablished!() -> (OutputDelegate, then, bHasValidMember:bool) {is=title K2Node_Event}
   > n8 = Branch(Condition=@n7.bHasValidMember)
   > n9 = SetNotReadyVisuals()
   n10 = SetEmptyVisuals(execute=@n8.else)
   Reroute Node()
   > n11 = PlayAnimationForward(InAnimation=Get StyleEmpty() -> (StyleEmpty:Widget Animation Object Reference))
  ]}
 SetEmptyVisuals {nodes [
   SetEmptyVisuals() {is=title K2Node_FunctionEntry}
   n1 = Get Image_Meeple() -> (Image_Meeple:Image Object Reference)
   > n2 = GetDynamicMaterial(self=@n1)
   > SetScalarParameterValue(self=@n2, ParameterName=UsesOutline, Value=1)
   > SetColorAndOpacity(self:Image Object Reference=@n1, InColorAndOpacity:Linear Color Structure=(R=0.254,G=0.496,B=0.791,A=1))
  ]}
 SetNotReadyVisuals {nodes [
   SetNotReadyVisuals() {is=title K2Node_FunctionEntry}
   n1 = Get Image_Meeple()
   > n2 = GetDynamicMaterial(self=@n1)
   > SetScalarParameterValue(self=@n2, ParameterName=UsesOutline, Value=0)
   > SetColorAndOpacity(self=@n1, InColorAndOpacity=(R=0.445,G=1,B=1,A=1))
  ]}
 SetReadyVisuals {nodes [SetReadyVisuals() {is=title K2Node_FunctionEntry}]}
 SetSitOutVisuals {nodes [
   SetSitOutVisuals() {is=title K2Node_FunctionEntry}
   n1 = Get Image_Meeple()
   > n2 = GetDynamicMaterial(self=@n1)
   > SetScalarParameterValue(self=@n2, ParameterName=UsesOutline, Value=0)
   > SetColorAndOpacity(self=@n1, InColorAndOpacity=(R=0.254,G=0.496,B=0.791,A=1))
  ]}
]
variables []
widgets {
 info {inheritedWidgetCount=2, parentClass=/Script/FortniteUI.AthenaMinimalSquadMemberEntry, rootWidgetClass=&Overlay, widgetCount=7}
 widgets [
  Overlay_0 {widgetClassPath=&Overlay, widget=Overlay_0}
  MeepleContainer {
   widgetClassPath=&Overlay
   bIsVariable=true
   parent=Overlay_0
   widget=MeepleContainer
   slot {class=OverlaySlot, properties {horizontalAlignment=HAlign_Center, verticalAlignment=VAlign_Center, padding=<ltrb -3 0 -3 0>}}
  }
  Image_Meeple {
   widgetClassPath=&Image
   bIsVariable=true
   parent=MeepleContainer
   widget=Image_Meeple
   properties {
    renderOpacity=0.8
    visibility=SelfHitTestInvisible
    brush {
     bIsDynamicallyLoaded=false
     drawAs=Image
     imageType=NoImage
     mirroring=NoMirror
     resourceName=None
     resourceObject=/FortniteCoreUI/UI/SocialPanel/Art/Materials/M_UI_SquadMember_Inst
     tiling=NoTile
     imageSize=<28 28>
     margin=<ltrb 0 0 0 0>
     outlineSettings {
      bUseBrushTransparency=false
      roundingType=HalfHeightRadius
      width=0
      color {colorUseRule=UseColor_Specified, specifiedColor=<rgba 0 0 0 0>}
      cornerRadii=<0 0 0 0>
     }
     tintColor {colorUseRule=UseColor_Specified, specifiedColor=<rgba 1 1 1 1>}
     uVRegion {bIsValid=false, max=<0 0>, min=<0 0>}
    }
   }
   slot {class=OverlaySlot, properties {horizontalAlignment=HAlign_Center, verticalAlignment=VAlign_Center}}
  }
  ReadyFlash {
   widgetClassPath=&Image
   bIsVariable=true
   parent=MeepleContainer
   widget=ReadyFlash
   properties {
    renderOpacity=0
    visibility=Collapsed
    brush {
     bIsDynamicallyLoaded=false
     drawAs=Image
     imageType=NoImage
     mirroring=NoMirror
     resourceName=None
     resourceObject=&M_UI_SquadMember_Inst
     tiling=NoTile
     imageSize=<28 28>
     margin=<ltrb 0 0 0 0>
     outlineSettings {
      bUseBrushTransparency=false
      roundingType=HalfHeightRadius
      width=0
      color {colorUseRule=UseColor_Specified, specifiedColor=<rgba 0 0 0 0>}
      cornerRadii=<0 0 0 0>
     }
     tintColor {colorUseRule=UseColor_Specified, specifiedColor=<rgba 1 1 1 1>}
     uVRegion {bIsValid=false, max=<0 0>, min=<0 0>}
    }
    colorAndOpacity=<rgba 0.0343 1 0.227 1>
   }
   slot {class=OverlaySlot, properties {horizontalAlignment=HAlign_Center, verticalAlignment=VAlign_Center}}
  }
  Image_ReadyIcon {
   widgetClassPath=&Image
   bInherited=true
   bIsVariable=true
   parent=Overlay_0
   widget=Image_ReadyIcon
   properties {
    brush {
     bIsDynamicallyLoaded=false
     drawAs=Image
     imageType=NoImage
     mirroring=NoMirror
     resourceName=None
     resourceObject=/Game/UI/SocialPanel/Art/Textures/T_UI_SquadCheck_32
     tiling=NoTile
     imageSize=<16 16>
     margin=<ltrb 0 0 0 0>
     outlineSettings {
      bUseBrushTransparency=false
      roundingType=HalfHeightRadius
      width=0
      color {colorUseRule=UseColor_Specified, specifiedColor=<rgba 0 0 0 0>}
      cornerRadii=<0 0 0 0>
     }
     tintColor {colorUseRule=UseColor_Specified, specifiedColor=<rgba 1 1 1 1>}
     uVRegion {bIsValid=false, max=<0 0>, min=<0 0>}
    }
    renderTransform {angle=0, scale=<1 1>, shear=<0 0>, translation=<2 2>}
   }
   slot {class=OverlaySlot, properties {horizontalAlignment=HAlign_Right, verticalAlignment=VAlign_Bottom}}
  }
  LocalUserIndicator {
   widgetClassPath=&Image
   bIsVariable=true
   parent=Overlay_0
   widget=LocalUserIndicator
   properties {
    visibility=Collapsed
    brush {
     bIsDynamicallyLoaded=false
     drawAs=Image
     imageType=NoImage
     mirroring=NoMirror
     resourceName=None
     resourceObject=/Game/UI/CodeReferenced/UI_QuickbarDownArrow
     tiling=NoTile
     imageSize=<16 16>
     margin=<ltrb 0 0 0 0>
     outlineSettings {
      bUseBrushTransparency=false
      roundingType=HalfHeightRadius
      width=0
      color {colorUseRule=UseColor_Specified, specifiedColor=<rgba 0 0 0 0>}
      cornerRadii=<0 0 0 0>
     }
     tintColor {colorUseRule=UseColor_Specified, specifiedColor=<rgba 1 1 1 1>}
     uVRegion {bIsValid=false, max=<0 0>, min=<0 0>}
    }
    colorAndOpacity=<rgba 0.9473 1 0.0704 1>
    renderTransform {angle=0, scale=<1 1>, shear=<0 0>, translation=<0 -12>}
   }
   slot {class=OverlaySlot, properties {horizontalAlignment=HAlign_Center}}
  }
  InputActionWidget {widgetClassPath=&CommonActionWidget, bInherited=true, widget=null}
 ]
}
