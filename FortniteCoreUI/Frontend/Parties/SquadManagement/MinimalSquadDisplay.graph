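# MinimalSquadDisplay is a squad entry widget (AthenaTeamSquadEntry child) that shows a squad's member entries above a colored ribbon.
# It collapses unless its squad has members and multiple squads are populated.
# On squad assignment, it async-loads the style's icon, then its ribbon, setting the ribbon texture and text color on SquadRibbon's dynamic material.
# The loaded icon is never displayed in this graph.
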WidgetBlueprint /FortniteCoreUI/Frontend/Parties/SquadManagement/MinimalSquadDisplay
graphs [EventGraph {nodes [
   n1 = Event OnSquadIsEmptyChanged!() -> (OutputDelegate, then, bSquadIsEmpty:bool, bMultipleSquadsPopulated:bool) {is=title K2Node_Event}
   n2 = NOT Boolean(A=@n1.bSquadIsEmpty)
   > SetVisibility(InVisibility=Select(Option 0=Collapsed, Option 1=HitTestInvisible, Index=AND Boolean(A=@n2, B=@n1.bMultipleSquadsPopulated)))
   n3 = Event OnSquadAssigned!() -> (OutputDelegate, then, SquadStyle:"Athena Team Display Info Structure (by ref)") {is=title K2Node_Event}
   n4 = Break Athena Team Display Info!(AthenaTeamDisplayInfo:"Athena Team Display Info Structure (by ref)"=@n3.SquadStyle) -> (Icon:Soft Object Reference, Ribbon:Soft Object Reference, Name:Text, TextColor:Linear Color Structure, PedestalColor:Linear Color Structure) {is=title K2Node_BreakStruct}
   > n5 = Async Load Asset(Asset=@n4)
   n6 = Async Load Asset(execute=@n5.Completed, Asset=@n4.Ribbon)
   n7 = Cast To Texture!(execute=@n6.Completed, Object=@n6) -> (then, CastFailed, AsTexture:Texture Object Reference) {is=title K2Node_DynamicCast}
   > n8 = GetDynamicMaterial(self=Get SquadRibbon() -> (SquadRibbon:Image Object Reference))
   > SetTextureParameterValue(self=@n8, ParameterName=RibbonTexture, Value=@n7)
   > SetVectorParameterValue(self=@n8, ParameterName=SquadRGB_2, Value=@n4.TextColor)
   Event PreConstruct()
   Event Tick()
   Event Construct()
  ]}]
variables []
widgets {
 info {inheritedWidgetCount=1, parentClass=/Script/FortniteUI.AthenaTeamSquadEntry, rootWidgetClass=&VerticalBox, widgetCount=3}
 widgets [
  VerticalBox_0 {widgetClassPath=&VerticalBox, widget=VerticalBox_0}
  EntryBox_Members {
   widgetClassPath=&DynamicEntryBox
   bInherited=true
   bIsVariable=true
   parent=VerticalBox_0
   widget=EntryBox_Members
   properties {entryWidgetClass=&/MinimalSquadMember.MinimalSquadMember_C, numDesignerPreviewEntries=4}
   slot {class=VerticalBoxSlot}
  }
  SquadRibbon {
   widgetClassPath=&Image
   bIsVariable=true
   parent=VerticalBox_0
   widget=SquadRibbon
   properties {
    visibility=SelfHitTestInvisible
    brush {
     bIsDynamicallyLoaded=false
     drawAs=Image
     imageType=NoImage
     mirroring=NoMirror
     resourceName=None
     resourceObject=/FortniteCoreUI/UI/SocialPanel/Art/Materials/M_UI_SquadColors_Ribbon_Master
     tiling=NoTile
     imageSize=<32 8>
     margin=<ltrb 0 0 0 0>
     outlineSettings {
      bUseBrushTransparency=false
      roundingType=HalfHeightRadius
      width=0
      color {colorUseRule=UseColor_Specified, specifiedColor=<rgba 0 0 0 0>}
      cornerRadii=<0 0 0 0>
     }
     tintColor {colorUseRule=UseColor_Specified, specifiedColor=<rgba 1 1 1 1>}
     uVRegion {bIsValid=false, max=<0 0>, min=<0 0>}
    }
   }
   slot {class=VerticalBoxSlot, properties {padding=<ltrb 0 4 0 0>}}
  }
 ]
}
