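# Squad member tile widget for squad management in the party frontend, showing avatar, name, status, leader icon and squad ribbon.
# When a member is established it plays an intro, applies not-ready or empty visuals, and pulses an arrow for the local player.
# Readiness and frontend changes feed SetVisuals, which restyles via dynamic material parameters; the Style animations are left unwired.
# Squad assignment async-loads the ribbon texture and tints ribbon and intro overlay materials.
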WidgetBlueprint /FortniteCoreUI/Frontend/Parties/SquadManagement/SquadMemberTile
animations [
 OnHovered {
  binding_count=3
  length_seconds=0.2
  object=&self:OnHovered
  sequence {bindings [
    EntryBorder {
     id=#1
     possessed_class=Border
     tracks [{
       type=MovieSceneWidgetMaterialTrack
       sections [{type=MovieSceneParameterSection, channels [AnimateHighlight {type=MovieSceneScriptingFloatChannel, keys [[0,0],[3,1],[4,1]]}]}]
      }]
    }
    EntryOvr {
     id=#2
     possessed_class=Overlay
     tracks [{
       type=MovieScene2DTransformTrack
       sections [{
         type=MovieScene2DTransformSection
         channels [
          Translation.X {type=MovieSceneScriptingFloatChannel, keys [[0,0],[2,-8]]}
          Translation.Y {type=MovieSceneScriptingFloatChannel}
          Angle {type=MovieSceneScriptingFloatChannel}
          Scale.X {type=MovieSceneScriptingFloatChannel}
          Scale.Y {type=MovieSceneScriptingFloatChannel}
          Shear.X {type=MovieSceneScriptingFloatChannel}
          Shear.Y {type=MovieSceneScriptingFloatChannel}
         ]
        }]
      }]
    }
    UserPortraitSB {
     id=#3
     possessed_class=SizeBox
     tracks [{
       type=MovieScene2DTransformTrack
       sections [{
         type=MovieScene2DTransformSection
         channels [
          Translation.X {type=MovieSceneScriptingFloatChannel, keys [[0,0],[2,-8]]}
          Translation.Y {type=MovieSceneScriptingFloatChannel}
          Angle {type=MovieSceneScriptingFloatChannel}
          Scale.X {type=MovieSceneScriptingFloatChannel}
          Scale.Y {type=MovieSceneScriptingFloatChannel}
          Shear.X {type=MovieSceneScriptingFloatChannel}
          Shear.Y {type=MovieSceneScriptingFloatChannel}
         ]
        }]
      }]
    }
   ]}
 }
 StyleSitOut {
  binding_count=4
  length_seconds=0.15
  object=&self:StyleSitOut
  sequence {bindings [
    Text_SocialName {
     id=#4
     possessed_class=CommonTextBlock
     tracks [{
       type=MovieSceneWidgetMaterialTrack
       display_name=Default Text Style Override.Font.FontMaterial
       sections [{
         type=MovieSceneParameterSection
         channels [
          PrimaryR {type=MovieSceneScriptingFloatChannel, const [0.2542,2]}
          PrimaryG {type=MovieSceneScriptingFloatChannel, const [0.4969,2]}
          PrimaryB {type=MovieSceneScriptingFloatChannel, const [0.7913,2]}
          PrimaryA {type=MovieSceneScriptingFloatChannel, const [1,2]}
         ]
        }]
      }]
    }
    Text_ParticipationStatus {
     id=#5
     possessed_class=CommonTextBlock
     tracks [{
       type=MovieSceneColorTrack
       display_name=Color and Opacity
       sections [{
         type=MovieSceneColorSection
         channels [
          Color.R {type=MovieSceneScriptingFloatChannel, const [0.0561,2]}
          Color.G {type=MovieSceneScriptingFloatChannel, const [0.2346,2]}
          Color.B {type=MovieSceneScriptingFloatChannel, const [0.5149,2]}
          Color.A {type=MovieSceneScriptingFloatChannel, const [1,2]}
         ]
        }]
      }]
    }
    SquadRibbon {
     id=#6
     possessed_class=Image
     tracks [{
       type=MovieSceneWidgetMaterialTrack
       sections [{type=MovieSceneParameterSection, channels [DimRibbon {type=MovieSceneScriptingFloatChannel, const [0.5,2]}]}]
      }]
    }
    AvatarIcon_MemberAvatar {
     id=#7
     possessed_class=FortSocialAvatarIcon
     tracks [{
       type=MovieSceneWidgetMaterialTrack
       sections [{
         type=MovieSceneParameterSection
         channels [
          IsInactive {type=MovieSceneScriptingFloatChannel, const [1,2]}
          IsEmpty {type=MovieSceneScriptingFloatChannel, const [0,2]}
         ]
        }]
      }]
    }
   ]}
 }
 StyleReady {
  binding_count=4
  length_seconds=0.15
  object=&self:StyleReady
  sequence {bindings [
    AvatarIcon_MemberAvatar {
     id=#8
     possessed_class=FortSocialAvatarIcon
     tracks [{
       type=MovieSceneWidgetMaterialTrack
       sections [{
         type=MovieSceneParameterSection
         channels [
          IsInactive {type=MovieSceneScriptingFloatChannel, const [0,2]}
          IsEmpty {type=MovieSceneScriptingFloatChannel, const [0,2]}
         ]
        }]
      }]
    }
    Text_SocialName {
     id=#9
     possessed_class=CommonTextBlock
     tracks [{
       type=MovieSceneWidgetMaterialTrack
       display_name=Default Text Style Override.Font.FontMaterial
       sections [{
         type=MovieSceneParameterSection
         channels [
          PrimaryR {type=MovieSceneScriptingFloatChannel, const [0.4452,2]}
          PrimaryG {type=MovieSceneScriptingFloatChannel, const [1,2]}
          PrimaryB {type=MovieSceneScriptingFloatChannel, const [1,2]}
          PrimaryA {type=MovieSceneScriptingFloatChannel, const [1,2]}
         ]
        }]
      }]
    }
    SquadRibbon {
     id=#10
     possessed_class=Image
     tracks [
      {
       type=MovieSceneFloatTrack
       display_name=Render Opacity
       sections [{type=MovieSceneFloatSection, channels [None {type=MovieSceneScriptingFloatChannel, const [1,2]}]}]
      }
      {
       type=MovieSceneWidgetMaterialTrack
       sections [{type=MovieSceneParameterSection, channels [DimRibbon {type=MovieSceneScriptingFloatChannel, const [0,2]}]}]
      }
     ]
    }
    Text_ParticipationStatus {
     id=#11
     possessed_class=CommonTextBlock
     tracks [{
       type=MovieSceneColorTrack
       display_name=Color and Opacity
       sections [{
         type=MovieSceneColorSection
         start_frame=0
         channels [
          Color.R {type=MovieSceneScriptingFloatChannel, const [0.3185,2]}
          Color.G {type=MovieSceneScriptingFloatChannel, const [1,2]}
          Color.B {type=MovieSceneScriptingFloatChannel, const [0.3185,2]}
          Color.A {type=MovieSceneScriptingFloatChannel, const [1,2]}
         ]
        }]
      }]
    }
   ]}
 }
 StyleNotReady {
  binding_count=4
  length_seconds=0.2
  object=&self:StyleNotReady
  sequence {bindings [
    SquadRibbon {
     id=#12
     possessed_class=Image
     tracks [
      {
       type=MovieSceneFloatTrack
       display_name=Render Opacity
       sections [{type=MovieSceneFloatSection, channels [None {type=MovieSceneScriptingFloatChannel, const [1,2]}]}]
      }
      {
       type=MovieSceneWidgetMaterialTrack
       sections [{type=MovieSceneParameterSection, channels [DimRibbon {type=MovieSceneScriptingFloatChannel, const [0,2]}]}]
      }
     ]
    }
    AvatarIcon_MemberAvatar {
     id=#8
     possessed_class=FortSocialAvatarIcon
     tracks [{
       type=MovieSceneWidgetMaterialTrack
       sections [{
         type=MovieSceneParameterSection
         channels [
          IsInactive {type=MovieSceneScriptingFloatChannel, const [0,2]}
          IsEmpty {type=MovieSceneScriptingFloatChannel, const [0,2]}
         ]
        }]
      }]
    }
    Text_SocialName {
     id=#9
     possessed_class=CommonTextBlock
     tracks [{
       type=MovieSceneWidgetMaterialTrack
       display_name=Default Text Style Override.Font.FontMaterial
       sections [{
         type=MovieSceneParameterSection
         channels [
          PrimaryR {type=MovieSceneScriptingFloatChannel, const [0.4452,2]}
          PrimaryG {type=MovieSceneScriptingFloatChannel, const [1,2]}
          PrimaryB {type=MovieSceneScriptingFloatChannel, const [1,2]}
          PrimaryA {type=MovieSceneScriptingFloatChannel, const [1,2]}
         ]
        }]
      }]
    }
    Text_ParticipationStatus {
     id=#13
     possessed_class=CommonTextBlock
     tracks [{
       type=MovieSceneColorTrack
       display_name=Color and Opacity
       sections [{
         type=MovieSceneColorSection
         channels [
          Color.R {type=MovieSceneScriptingFloatChannel, const [0.2542,2]}
          Color.G {type=MovieSceneScriptingFloatChannel, const [0.4969,2]}
          Color.B {type=MovieSceneScriptingFloatChannel, const [0.7913,2]}
          Color.A {type=MovieSceneScriptingFloatChannel, const [1,2]}
         ]
        }]
      }]
    }
   ]}
 }
 StyleEmpty {
  binding_count=4
  length_seconds=0.2
  object=&self:StyleEmpty
  sequence {bindings [
    Text_SocialName {
     id=#4
     possessed_class=CommonTextBlock
     tracks [{
       type=MovieSceneWidgetMaterialTrack
       display_name=Default Text Style Override.Font.FontMaterial
       sections [{
         type=MovieSceneParameterSection
         channels [
          PrimaryR {type=MovieSceneScriptingFloatChannel, const [0.0561,2]}
          PrimaryG {type=MovieSceneScriptingFloatChannel, const [0.2346,2]}
          PrimaryB {type=MovieSceneScriptingFloatChannel, const [0.5149,2]}
          PrimaryA {type=MovieSceneScriptingFloatChannel, const [1,2]}
         ]
        }]
      }]
    }
    SquadRibbon {
     id=#14
     possessed_class=Image
     tracks [{
       type=MovieSceneFloatTrack
       display_name=Render Opacity
       sections [{type=MovieSceneFloatSection, channels [None {type=MovieSceneScriptingFloatChannel, const [0,2]}]}]
      }]
    }
    Text_ParticipationStatus {
     id=#5
     possessed_class=CommonTextBlock
     tracks [{
       type=MovieSceneColorTrack
       display_name=Color and Opacity
       sections [{
         type=MovieSceneColorSection
         channels [
          Color.R {type=MovieSceneScriptingFloatChannel, const [0.0561,2]}
          Color.G {type=MovieSceneScriptingFloatChannel, const [0.2346,2]}
          Color.B {type=MovieSceneScriptingFloatChannel, const [0.5149,2]}
          Color.A {type=MovieSceneScriptingFloatChannel, const [1,2]}
         ]
        }]
      }]
    }
    AvatarIcon_MemberAvatar {
     id=#7
     possessed_class=FortSocialAvatarIcon
     tracks [{
       type=MovieSceneWidgetMaterialTrack
       sections [{type=MovieSceneParameterSection, channels [IsEmpty {type=MovieSceneScriptingFloatChannel, const [1,2]}]}]
      }]
    }
   ]}
 }
 SetRibbonRotation {
  binding_count=1
  length_seconds=0.2
  object=&self:SetRibbonRotation
  sequence {bindings [SquadRibbon {
     id=#15
     possessed_class=Image
     tracks [{
       type=MovieSceneWidgetMaterialTrack
       sections [{type=MovieSceneParameterSection, channels [RibbonStripeRotation {type=MovieSceneScriptingFloatChannel, const [0.25,2]}]}]
      }]
    }]}
 }
 OnSelected {
  binding_count=1
  length_seconds=1.2
  object=&self:OnSelected
  sequence {bindings [LocalUserIcon {
     id=#16
     possessed_class=Image
     tracks [
      {
       type=MovieSceneByteTrack
       display_name=Visibility
       sections [{type=MovieSceneByteSection, channels [None {type=MovieSceneScriptingByteChannel, const [4,1]}]}]
      }
      {
       type=MovieSceneColorTrack
       display_name=Color and Opacity
       sections [{
         type=MovieSceneColorSection
         channels [
          Color.R {type=MovieSceneScriptingFloatChannel, const [1,1]}
          Color.G {type=MovieSceneScriptingFloatChannel, const [1,1]}
          Color.B {type=MovieSceneScriptingFloatChannel, const [1,1]}
          Color.A {type=MovieSceneScriptingFloatChannel, const [1,1]}
         ]
        }]
      }
      {
       type=MovieScene2DTransformTrack
       sections [{
         type=MovieScene2DTransformSection
         channels [
          Translation.X {type=MovieSceneScriptingFloatChannel, keys [[0,-28],[8,-20],[16,-20],[24,-28]]}
          Translation.Y {type=MovieSceneScriptingFloatChannel}
          Angle {type=MovieSceneScriptingFloatChannel}
          Scale.X {type=MovieSceneScriptingFloatChannel, keys [[0,1],[8,1.25],[16,1.25],[24,1]]}
          Scale.Y {type=MovieSceneScriptingFloatChannel, keys [[0,1],[8,1.25],[16,1.25],[24,1]]}
          Shear.X {type=MovieSceneScriptingFloatChannel}
          Shear.Y {type=MovieSceneScriptingFloatChannel}
         ]
        }]
      }
     ]
    }]}
 }
 OnLocalUserDefined {
  binding_count=1
  length_seconds=1.2
  object=&self:OnLocalUserDefined
  sequence {bindings [LocalUserIcon {
     id=#17
     possessed_class=Image
     tracks [
      {
       type=MovieScene2DTransformTrack
       sections [{
         type=MovieScene2DTransformSection
         channels [
          Translation.X {type=MovieSceneScriptingFloatChannel, keys [[0,-28],[8,-20],[16,-20],[24,-28]]}
          Translation.Y {type=MovieSceneScriptingFloatChannel}
          Angle {type=MovieSceneScriptingFloatChannel}
          Scale.X {type=MovieSceneScriptingFloatChannel, keys [[0,1],[8,1.25],[16,1.25],[24,1]]}
          Scale.Y {type=MovieSceneScriptingFloatChannel, keys [[0,1],[8,1.25],[16,1.25],[24,1]]}
          Shear.X {type=MovieSceneScriptingFloatChannel}
          Shear.Y {type=MovieSceneScriptingFloatChannel}
         ]
        }]
      }
      {
       type=MovieSceneColorTrack
       display_name=Color and Opacity
       sections [{
         type=MovieSceneColorSection
         channels [
          Color.R {type=MovieSceneScriptingFloatChannel, const [0.9473,1]}
          Color.G {type=MovieSceneScriptingFloatChannel, const [1,1]}
          Color.B {type=MovieSceneScriptingFloatChannel, const [0.0704,1]}
          Color.A {type=MovieSceneScriptingFloatChannel, const [1,1]}
         ]
        }]
      }
     ]
    }]}
 }
 OnIntro {
  binding_count=3
  length_seconds=0.4
  object=&self:OnIntro
  sequence {bindings [
    PrimaryOvr {
     id=#18
     possessed_class=Overlay
     tracks [{
       type=MovieScene2DTransformTrack
       sections [{
         type=MovieScene2DTransformSection
         channels [
          Translation.X {type=MovieSceneScriptingFloatChannel, keys [[0,-128],[8,0]]}
          Translation.Y {type=MovieSceneScriptingFloatChannel}
          Angle {type=MovieSceneScriptingFloatChannel}
          Scale.X {type=MovieSceneScriptingFloatChannel}
          Scale.Y {type=MovieSceneScriptingFloatChannel}
          Shear.X {type=MovieSceneScriptingFloatChannel}
          Shear.Y {type=MovieSceneScriptingFloatChannel}
         ]
        }]
      }]
    }
    ContentHB {
     id=#19
     possessed_class=HorizontalBox
     tracks [{
       type=MovieSceneFloatTrack
       display_name=Render Opacity
       sections [{type=MovieSceneFloatSection, channels [None {type=MovieSceneScriptingFloatChannel, keys [[0,0],[3,0],[6,1]]}]}]
      }]
    }
    SquadOverlay {
     id=#20
     possessed_class=Image
     tracks [
      {
       type=MovieSceneFloatTrack
       display_name=Render Opacity
       sections [{type=MovieSceneFloatSection, channels [None {type=MovieSceneScriptingFloatChannel, keys [[0,0],[3,0.7],[8,0]]}]}]
      }
      {
       type=MovieSceneByteTrack
       display_name=Visibility
       sections [{type=MovieSceneByteSection, channels [None {type=MovieSceneScriptingByteChannel, keys [[0,4],[8,1]]}]}]
      }
     ]
    }
   ]}
 }
]
graphs [
 EventGraph {nodes [
   n1 = PlayAnimationForward(InAnimation=Get StyleNotReady() -> (StyleNotReady:Widget Animation Object Reference))
   n2 = Branch(execute=[@n1.then, @n15.then], Condition=IsCurrentMemberLocal!(self:Fort Team Member Entry Base Object Reference) -> (ReturnValue:bool))
   n3 = Get LocalUserIcon() -> (LocalUserIcon:Image Object Reference)
   > SetVisibility(self=@n3, InVisibility=SelfHitTestInvisible)
   > PlayAnimation(InAnimation=Get OnLocalUserDefined() -> (OnLocalUserDefined:Widget Animation Object Reference), NumLoopsToPlay=0)
   SetVisibility(execute=@n2.else, self=@n3, InVisibility=Collapsed)
   SetVisibility(self=Get LocalUserIcon(), InVisibility=SelfHitTestInvisible)
   > PlayAnimation(InAnimation=Get OnSelected() -> (OnSelected:Widget Animation Object Reference), NumLoopsToPlay=0)
   n4 = PlayAnimationForward(InAnimation=Get StyleEmpty() -> (StyleEmpty:Widget Animation Object Reference))
   SetVisibility(execute=[@n16.then, @n4.then], self=@n3, InVisibility=Collapsed)
   n5 = Event Construct()
   PlayAnimationForward(execute=[@n5.then, @n6.then], InAnimation=Get SetRibbonRotation() -> (SetRibbonRotation:Widget Animation Object Reference))
   n6 = Event PreConstruct()
   Event OnSquadAssigned() {is=title K2Node_Event}
   n7 = Break Athena Team Display Info!(AthenaTeamDisplayInfo:"Athena Team Display Info Structure (by ref)"=GetSquadDisplayInfo!(self:Athena Squad Member Detailed Entry Object Reference) -> (ReturnValue:Athena Team Display Info Structure)) -> (Icon:Soft Object Reference, Ribbon:Soft Object Reference, Name:Text, TextColor:Linear Color Structure, PedestalColor:Linear Color Structure) {is=title K2Node_BreakStruct}
   > n8 = Async Load Asset(Asset=@n7.Ribbon)
   n9 = Cast To Texture!(execute=@n8.Completed, Object=@n8) -> (then, CastFailed, AsTexture:Texture Object Reference) {is=title K2Node_DynamicCast}
   > n10 = GetDynamicMaterial(self=Get SquadRibbon() -> (SquadRibbon:Image Object Reference))
   > SetTextureParameterValue(self=@n10, ParameterName=RibbonTexture, Value=@n9)
   > SetVectorParameterValue(self=@n10, ParameterName=SquadRGB_2, Value=@n7.TextColor)
   > n11 = GetDynamicMaterial(self=Get SquadOverlay() -> (SquadOverlay:Image Object Reference))
   > SetTextureParameterValue(self=@n11, ParameterName=RibbonTexture, Value=@n9)
   > SetVectorParameterValue(self=@n11, ParameterName=SquadRGB_2, Value=@n7.TextColor)
   Event On Unhovered()
   > PlayAnimation(InAnimation=Get OnHovered() -> (OnHovered:Widget Animation Object Reference), PlayMode=Reverse, PlaybackSpeed=2)
   n12 = Event BP_OnTeamMemberEstablished!() -> (OutputDelegate, then, bHasValidMember:bool) {is=title K2Node_Event}
   > PlayAnimationForward(InAnimation=Get OnIntro() -> (OnIntro:Widget Animation Object Reference))
   > n13 = Branch(Condition=@n12.bHasValidMember)
   n14 = Get EntryBorder() -> (EntryBorder:Border Object Reference)
   > SetBrushFromMaterial(self:Border Object Reference=@n14, Material=/Game/UI/SocialPanel/Art/Materials/M_UI_SocialUserItemBorder_UserListItem)
   > n15 = SetNotReadyVisuals()
   SetBrushFromMaterial(execute=@n13.else, self=@n14, Material=/Game/UI/SocialPanel/Art/Materials/M_UI_SocialUserItemBorder_InactiveUserListItem)
   > n16 = SetEmptyVisuals()
   Event On Hovered()
   > PlayAnimation(InAnimation=Get OnHovered())
   n17 = Event OnMemberGameReadinessChanged!() -> (OutputDelegate, then, ReadyStatus:EGameReadiness Enum=NotReady) {is=title K2Node_Event}
   > SetVisuals!(
    execute
    self
    Member Readiness:EGameReadiness Enum=@n17.ReadyStatus
    Is In Frontend:bool=IsCurrentMemberInFrontend!(self:Fort Team Member Entry Base Object Reference) -> (ReturnValue:bool)
   ) -> (then)
   n18 = Event OnMemberIsInFrontendChanged!() -> (OutputDelegate, then, bIsInFrontend:bool) {is=title K2Node_Event}
   > SetVisuals(
    Member Readiness=GetCurrentMemberGameReadiness!(self:Fort Team Member Entry Base Object Reference) -> (ReturnValue:EGameReadiness Enum=NotReady)
    Is In Frontend=@n18.bIsInFrontend
   )
   SetVisibility(self=Get LocalUserIcon(), InVisibility=Collapsed)
   Event On Selected()
   Event On Deselected()
  ]}
 SetEmptyVisuals {nodes [
   SetEmptyVisuals() {is=title K2Node_FunctionEntry}
   > n1 = GetDynamicMaterial(self=Get AvatarIcon_MemberAvatar() -> (AvatarIcon_MemberAvatar:Fort Social Avatar Icon Object Reference))
   > SetScalarParameterValue(self=@n1, ParameterName=IsEmpty, Value=1)
   > SetScalarParameterValue(self=@n1, ParameterName=IsInactive, Value=0)
   > SetColorAndOpacity(
    self=Get Text_ParticipationStatus() -> (Text_ParticipationStatus:Common Text Object Reference)
    InColorAndOpacity=Make Slate Color(SpecifiedColor=(R=0.056,G=0.234,B=0.514,A=1))
   )
   n2 = Get SquadRibbon()
   > n3 = GetDynamicMaterial(self=@n2)
   > SetScalarParameterValue(self=@n3, ParameterName=DimRibbon, Value=0)
   > SetRenderOpacity(self=@n2, InOpacity=0)
  ]}
 SetNotReadyVisuals {nodes [
   SetNotReadyVisuals() {is=title K2Node_FunctionEntry}
   > n1 = GetDynamicMaterial(self=Get AvatarIcon_MemberAvatar())
   > SetScalarParameterValue(self=@n1, ParameterName=IsEmpty, Value=0)
   > SetScalarParameterValue(self=@n1, ParameterName=IsInactive, Value=0)
   > SetColorAndOpacity(self=Get Text_ParticipationStatus(), InColorAndOpacity=Make Slate Color(SpecifiedColor=(R=0.254,G=0.496,B=0.791,A=1)))
   n2 = Get SquadRibbon()
   > n3 = GetDynamicMaterial(self=@n2)
   > SetScalarParameterValue(self=@n3, ParameterName=DimRibbon, Value=0)
   > SetRenderOpacity(self=@n2, InOpacity=1)
  ]}
 SetReadyVisuals {nodes [
   SetReadyVisuals() {is=title K2Node_FunctionEntry}
   > n1 = GetDynamicMaterial(self=Get AvatarIcon_MemberAvatar())
   > SetScalarParameterValue(self=@n1, ParameterName=IsEmpty, Value=0)
   > SetScalarParameterValue(self=@n1, ParameterName=IsInactive, Value=0)
   > SetColorAndOpacity(self=Get Text_ParticipationStatus(), InColorAndOpacity=Make Slate Color(SpecifiedColor=(R=0.31,G=1,B=0.31,A=1)))
   n2 = Get SquadRibbon()
   > n3 = GetDynamicMaterial(self=@n2)
   > SetScalarParameterValue(self=@n3, ParameterName=DimRibbon, Value=0)
   > SetRenderOpacity(self=@n2, InOpacity=1)
  ]}
 SetSitOutVisuals {nodes [
   SetSitOutVisuals() {is=title K2Node_FunctionEntry}
   > n1 = GetDynamicMaterial(self=Get AvatarIcon_MemberAvatar())
   > SetScalarParameterValue(self=@n1, ParameterName=IsEmpty, Value=0)
   > SetScalarParameterValue(self=@n1, ParameterName=IsInactive, Value=1)
   > SetColorAndOpacity(self=Get Text_ParticipationStatus(), InColorAndOpacity=Make Slate Color(SpecifiedColor=(R=0.056,G=0.234,B=0.514,A=1)))
   n2 = Get SquadRibbon()
   > n3 = GetDynamicMaterial(self=@n2)
   > SetScalarParameterValue(self=@n3, ParameterName=DimRibbon, Value=0.5)
   > SetRenderOpacity(self=@n2, InOpacity=1)
  ]}
 SetVisuals {nodes [
   n1 = SetVisuals!() -> (then, Member Readiness:EGameReadiness Enum=NotReady, Is In Frontend:bool=~) {is=title K2Node_FunctionEntry}
   > SetBrushFromMaterial(
    self=Get EntryBorder()
    Material=Select!(
     NotReady:Material Interface Object Reference=&M_UI_SocialUserItemBorder_UserListItem
     Ready:Material Interface Object Reference=&M_UI_SocialUserItemBorder_UserListItem
     Reevaluating:Material Interface Object Reference
     SittingOut:Material Interface Object Reference=&M_UI_SocialUserItemBorder_InactiveUserListItem
     Away:Material Interface Object Reference=&M_UI_SocialUserItemBorder_InactiveUserListItem
     Index:EGameReadiness Enum=@n1
    ) -> (ReturnValue:Material Interface Object Reference)
   )
   > n2 = Branch(Condition=OR Boolean(A=NOT Boolean(A=@n1.Is In Frontend), B="Equal (Enum)"(A:EGameReadiness Enum=@n1, B:EGameReadiness Enum=NotReady)))
   > SetNotReadyVisuals() {is=title K2Node_CallFunction}
   n3 = Branch(execute=@n2.else, Condition="Equal (Enum)"(A=@n1, B=Ready))
   > SetReadyVisuals() {is=title K2Node_CallFunction}
   Branch(execute=@n3.else, Condition=OR Boolean(A="Equal (Enum)"(A=@n1, B=SittingOut), B="Equal (Enum)"(A=@n1, B=Away)))
   > SetSitOutVisuals() {is=title K2Node_CallFunction}
  ]}
]
variables []
widgets {
 info {inheritedWidgetCount=8, parentClass=/Script/FortniteUI.AthenaSquadMemberDetailedEntry, rootWidgetClass=&SizeBox, widgetCount=24}
 widgets [
  SizeBox_0 {
   widgetClassPath=&SizeBox
   widget=SizeBox_0
   properties {bOverride_HeightOverride=true, bOverride_MinDesiredWidth=true, heightOverride=96, minDesiredWidth=400}
  }
  PrimaryOvr {widgetClassPath=&Overlay, parent=SizeBox_0, widget=PrimaryOvr, slot {class=SizeBoxSlot}}
  ContentHB {
   widgetClassPath=&HorizontalBox
   parent=PrimaryOvr
   widget=ContentHB
   slot {class=OverlaySlot, properties {horizontalAlignment=HAlign_Fill, verticalAlignment=VAlign_Fill}}
  }
  UserPortraitSB {
   widgetClassPath=&SizeBox
   parent=ContentHB
   widget=UserPortraitSB
   properties {bOverride_WidthOverride=true, widthOverride=96}
   slot {class=HorizontalBoxSlot}
  }
  Overlay_0 {widgetClassPath=&Overlay, parent=UserPortraitSB, widget=Overlay_0, slot {class=SizeBoxSlot}}
  LoadingThrobberBacking {
   widgetClassPath=&Image
   bIsVariable=true
   parent=Overlay_0
   widget=LoadingThrobberBacking
   properties {visibility=SelfHitTestInvisible, colorAndOpacity=<rgba 0 0.0999 0.3095 1>}
   slot {class=OverlaySlot, properties {horizontalAlignment=HAlign_Fill, verticalAlignment=VAlign_Fill}}
  }
  AvatarIcon_MemberAvatar {
   widgetClassPath=/Script/FortniteUI.FortSocialAvatarIcon
   bInherited=true
   bIsVariable=true
   parent=Overlay_0
   widget=AvatarIcon_MemberAvatar
   properties {
    avatarMaterial=/FortniteCoreUI/UI/SocialPanel/Art/Materials/UI_UserPortrait_Bust
    materialTextureParamName=LargePreview
    brush {
     bIsDynamicallyLoaded=false
     drawAs=Image
     imageType=NoImage
     mirroring=NoMirror
     resourceName=None
     resourceObject=&UI_UserPortrait_Bust
     tiling=NoTile
     imageSize=<32 32>
     margin=<ltrb 0 0 0 0>
     outlineSettings {
      bUseBrushTransparency=false
      roundingType=HalfHeightRadius
      width=0
      color {colorUseRule=UseColor_Specified, specifiedColor=<rgba 0 0 0 0>}
      cornerRadii=<0 0 0 0>
     }
     tintColor {colorUseRule=UseColor_Specified, specifiedColor=<rgba 1 1 1 1>}
     uVRegion {bIsValid=false, max=<0 0>, min=<0 0>}
    }
   }
   slot {class=OverlaySlot, properties {horizontalAlignment=HAlign_Fill, verticalAlignment=VAlign_Fill}}
  }
  HorizontalBox_0 {
   widgetClassPath=&HorizontalBox
   parent=Overlay_0
   widget=HorizontalBox_0
   slot {class=OverlaySlot, properties {horizontalAlignment=HAlign_Fill, verticalAlignment=VAlign_Fill, padding=<ltrb -5 0 -5 0>}}
  }
  LocalUserIcon {
   widgetClassPath=&Image
   bIsVariable=true
   parent=HorizontalBox_0
   widget=LocalUserIcon
   properties {
    bFlipForRightToLeftFlowDirection=true
    visibility=Collapsed
    brush {
     bIsDynamicallyLoaded=false
     drawAs=Image
     imageType=NoImage
     mirroring=NoMirror
     resourceName=None
     resourceObject=/Game/UI/SocialPanel/Art/Textures/T_UI_Icon_TreeArrow_64
     tiling=NoTile
     imageSize=<40 40>
     margin=<ltrb 0 0 0 0>
     outlineSettings {
      bUseBrushTransparency=false
      roundingType=HalfHeightRadius
      width=0
      color {colorUseRule=UseColor_Specified, specifiedColor=<rgba 0 0 0 0>}
      cornerRadii=<0 0 0 0>
     }
     tintColor {colorUseRule=UseColor_Specified, specifiedColor=<rgba 1 1 1 1>}
     uVRegion {bIsValid=false, max=<0 0>, min=<0 0>}
    }
    colorAndOpacity=<rgba 0.9473 1 0.0704 1>
   }
   slot {class=HorizontalBoxSlot, properties {horizontalAlignment=HAlign_Center, verticalAlignment=VAlign_Center}}
  }
  Spacing {
   widgetClassPath=&Border
   parent=HorizontalBox_0
   widget=Spacing
   properties {brushColor=<rgba 1 1 1 0>, padding=<ltrb 0 2 0 2>}
   slot {class=HorizontalBoxSlot, properties {size {sizeRule=Fill, value=1}}}
  }
  EntryOvr {
   widgetClassPath=&Overlay
   parent=ContentHB
   widget=EntryOvr
   slot {class=HorizontalBoxSlot, properties {size {sizeRule=Fill, value=1}}}
  }
  EntryBorder {
   widgetClassPath=&Border
   bIsVariable=true
   parent=EntryOvr
   widget=EntryBorder
   properties {
    verticalAlignment=VAlign_Center
    background {
     bIsDynamicallyLoaded=false
     drawAs=Image
     imageType=NoImage
     mirroring=NoMirror
     resourceName=None
     resourceObject=&M_UI_SocialUserItemBorder_UserListItem
     tiling=NoTile
     imageSize=<32 32>
     margin=<ltrb 0 0 0 0>
     outlineSettings {
      bUseBrushTransparency=false
      roundingType=HalfHeightRadius
      width=0
      color {colorUseRule=UseColor_Specified, specifiedColor=<rgba 0 0 0 0>}
      cornerRadii=<0 0 0 0>
     }
     tintColor {colorUseRule=UseColor_Specified, specifiedColor=<rgba 1 1 1 1>}
     uVRegion {bIsValid=false, max=<0 0>, min=<0 0>}
    }
    padding=<ltrb 16 0 20 0>
   }
   slot {class=OverlaySlot, properties {horizontalAlignment=HAlign_Fill, verticalAlignment=VAlign_Fill}}
  }
  VerticalBox_0 {
   widgetClassPath=&VerticalBox
   parent=EntryBorder
   widget=VerticalBox_0
   slot {class=BorderSlot, properties {verticalAlignment=VAlign_Center, padding=<ltrb 16 0 20 0>}}
  }
  HorizontalBox_2 {
   widgetClassPath=&HorizontalBox
   parent=VerticalBox_0
   widget=HorizontalBox_2
   slot {class=VerticalBoxSlot, properties {horizontalAlignment=HAlign_Left, size {sizeRule=Fill, value=1}}}
  }
  Text_SocialName {
   widgetClassPath=/Script/FortniteUI.FortSocialNameTextBlock
   bInherited=true
   parent=HorizontalBox_2
   widget=Text_SocialName
   properties {
    bOverrideDefaultStyle=true
    bTintInlineIcon=true
    scrollStyle=/FortUILibrary/Styles/Typography/Foundation/TextScrollStyle-NoFade.TextScrollStyle-NoFade_C
    text="User Name<PC/> <text color=\"1DEFA9FF\" size=\"12\">(NickName)</>"
    defaultTextStyleOverride {
     overflowPolicy=Clip
     transformPolicy=None
     colorAndOpacity {colorUseRule=UseColor_Specified, specifiedColor=<rgba 0.4452 1 1 1>}
     font {
      bForceMonospaced=false
      bMaterialIsStencil=false
      fontMaterial=null
      fontObject=&BurbankSmall
      letterSpacing=0
      monospacedWidth=1
      size=16
      skewAmount=0
      typefaceFontName=Black
      outlineSettings {
       bApplyOutlineToDropShadows=false
       bMiteredCorners=false
       bSeparateFillAlpha=false
       outlineMaterial=null
       outlineSize=0
       outlineColor=<rgba 0 0 0 1>
      }
     }
     highlightColor {colorUseRule=UseColor_Specified, specifiedColor=<rgba 0 0 0 1>}
     highlightShape {
      bIsDynamicallyLoaded=false
      drawAs=Image
      imageType=NoImage
      mirroring=NoMirror
      resourceName=None
      resourceObject=null
      tiling=NoTile
      imageSize=<32 32>
      margin=<ltrb 0 0 0 0>
      outlineSettings {
       bUseBrushTransparency=false
       roundingType=HalfHeightRadius
       width=0
       color {colorUseRule=UseColor_Specified, specifiedColor=<rgba 0 0 0 0>}
       cornerRadii=<0 0 0 0>
      }
      tintColor {colorUseRule=UseColor_Specified, specifiedColor=<rgba 1 1 1 1>}
      uVRegion {bIsValid=false, max=<0 0>, min=<0 0>}
     }
     selectedBackgroundColor {colorUseRule=UseColor_Foreground, specifiedColor=<rgba 1 0 1 1>}
     shadowColorAndOpacity=<rgba 0 0 0 1>
     shadowOffset=<0 0>
     strikeBrush {
      bIsDynamicallyLoaded=false
      drawAs=Image
      imageType=NoImage
      mirroring=NoMirror
      resourceName=None
      resourceObject=null
      tiling=NoTile
      imageSize=<32 32>
      margin=<ltrb 0 0 0 0>
      outlineSettings {
       bUseBrushTransparency=false
       roundingType=HalfHeightRadius
       width=0
       color {colorUseRule=UseColor_Specified, specifiedColor=<rgba 0 0 0 0>}
       cornerRadii=<0 0 0 0>
      }
      tintColor {colorUseRule=UseColor_Specified, specifiedColor=<rgba 1 1 1 1>}
      uVRegion {bIsValid=false, max=<0 0>, min=<0 0>}
     }
     underlineBrush {
      bIsDynamicallyLoaded=false
      drawAs=Image
      imageType=NoImage
      mirroring=NoMirror
      resourceName=None
      resourceObject=null
      tiling=NoTile
      imageSize=<32 32>
      margin=<ltrb 0 0 0 0>
      outlineSettings {
       bUseBrushTransparency=false
       roundingType=HalfHeightRadius
       width=0
       color {colorUseRule=UseColor_Specified, specifiedColor=<rgba 0 0 0 0>}
       cornerRadii=<0 0 0 0>
      }
      tintColor {colorUseRule=UseColor_Specified, specifiedColor=<rgba 1 1 1 1>}
      uVRegion {bIsValid=false, max=<0 0>, min=<0 0>}
     }
    }
   }
   slot {class=HorizontalBoxSlot, properties {verticalAlignment=VAlign_Center, size {sizeRule=Fill, value=1}}}
  }
  Image_LeaderIcon {
   widgetClassPath=&Image
   bInherited=true
   parent=HorizontalBox_2
   widget=Image_LeaderIcon
   properties {
    brush {
     bIsDynamicallyLoaded=false
     drawAs=Image
     imageType=NoImage
     mirroring=NoMirror
     resourceName=None
     resourceObject=/Game/UI/Foundation/Banner/M-CrownShine
     tiling=NoTile
     imageSize=<24 16>
     margin=<ltrb 0 0 0 0>
     outlineSettings {
      bUseBrushTransparency=false
      roundingType=HalfHeightRadius
      width=0
      color {colorUseRule=UseColor_Specified, specifiedColor=<rgba 0 0 0 0>}
      cornerRadii=<0 0 0 0>
     }
     tintColor {colorUseRule=UseColor_Specified, specifiedColor=<rgba 1 1 1 1>}
     uVRegion {bIsValid=false, max=<0 0>, min=<0 0>}
    }
    renderTransform {angle=0, scale=<1 1>, shear=<0 0>, translation=<0 2>}
   }
   slot {class=HorizontalBoxSlot, properties {verticalAlignment=VAlign_Center, padding=<ltrb 8 0 8 0>}}
  }
  Image_PlatformIcon {
   widgetClassPath=&Image
   bInherited=true
   bIsVariable=true
   parent=HorizontalBox_2
   widget=Image_PlatformIcon
   properties {visibility=Collapsed}
   slot {class=HorizontalBoxSlot, properties {horizontalAlignment=HAlign_Center, verticalAlignment=VAlign_Center}}
  }
  SizeBox_1 {
   widgetClassPath=&SizeBox
   parent=VerticalBox_0
   widget=SizeBox_1
   slot {class=VerticalBoxSlot, properties {horizontalAlignment=HAlign_Left, padding=<ltrb 0 8 0 0>}}
  }
  Text_ParticipationStatus {
   widgetClassPath=&CommonTextBlock
   bInherited=true
   bIsVariable=true
   parent=SizeBox_1
   widget=Text_ParticipationStatus
   properties {
    text=Status
    textTransformPolicy=ToUpper
    visibility=Collapsed
    colorAndOpacity {colorUseRule=UseColor_Specified, specifiedColor=<rgba 0.2542 0.4969 0.7913 1>}
    font {
     bForceMonospaced=false
     bMaterialIsStencil=false
     fontMaterial=null
     fontObject=&BurbankSmall
     letterSpacing=0
     monospacedWidth=1
     size=15
     skewAmount=0
     typefaceFontName=Bold
     outlineSettings {
      bApplyOutlineToDropShadows=false
      bMiteredCorners=false
      bSeparateFillAlpha=false
      outlineMaterial=null
      outlineSize=0
      outlineColor=<rgba 0 0 0 1>
     }
    }
    shadowOffset=<0 0>
   }
   slot {class=SizeBoxSlot}
  }
  SquadRibbon {
   widgetClassPath=&Image
   bIsVariable=true
   parent=EntryOvr
   widget=SquadRibbon
   properties {
    visibility=SelfHitTestInvisible
    brush {
     bIsDynamicallyLoaded=false
     drawAs=Image
     imageType=NoImage
     mirroring=NoMirror
     resourceName=None
     resourceObject=/FortniteCoreUI/UI/SocialPanel/Art/Materials/M_UI_SquadColors_Ribbon_Master
     tiling=NoTile
     imageSize=<12 32>
     margin=<ltrb 0 0 0 0>
     outlineSettings {
      bUseBrushTransparency=false
      roundingType=HalfHeightRadius
      width=0
      color {colorUseRule=UseColor_Specified, specifiedColor=<rgba 0 0 0 0>}
      cornerRadii=<0 0 0 0>
     }
     tintColor {colorUseRule=UseColor_Specified, specifiedColor=<rgba 1 1 1 1>}
     uVRegion {bIsValid=false, max=<0 0>, min=<0 0>}
    }
   }
   slot {class=OverlaySlot, properties {horizontalAlignment=HAlign_Right, verticalAlignment=VAlign_Fill}}
  }
  SquadOverlay {
   widgetClassPath=&Image
   bIsVariable=true
   parent=PrimaryOvr
   widget=SquadOverlay
   properties {
    renderOpacity=0
    visibility=Collapsed
    brush {
     bIsDynamicallyLoaded=false
     drawAs=Image
     imageType=NoImage
     mirroring=NoMirror
     resourceName=None
     resourceObject=&M_UI_SquadColors_Ribbon_Master
     tiling=NoTile
     imageSize=<12 32>
     margin=<ltrb 0 0 0 0>
     outlineSettings {
      bUseBrushTransparency=false
      roundingType=HalfHeightRadius
      width=0
      color {colorUseRule=UseColor_Specified, specifiedColor=<rgba 0 0 0 0>}
      cornerRadii=<0 0 0 0>
     }
     tintColor {colorUseRule=UseColor_Specified, specifiedColor=<rgba 1 1 1 1>}
     uVRegion {bIsValid=false, max=<0 0>, min=<0 0>}
    }
   }
   slot {class=OverlaySlot, properties {horizontalAlignment=HAlign_Fill, verticalAlignment=VAlign_Fill}}
  }
  Banner_MemberBanner {widgetClassPath=/Script/FortniteUI.FortPlayerBanner, bInherited=true, widget=null}
  VoiceIcon_Status {widgetClassPath=/Script/FortniteUI.FortVoiceChatStatusIcon, bInherited=true, widget=null}
  InputActionWidget {widgetClassPath=&CommonActionWidget, bInherited=true, widget=null}
 ]
}
