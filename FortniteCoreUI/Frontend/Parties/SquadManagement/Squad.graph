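# Squad entry widget for party squad management, showing a squad's logo, ribbon, name and member tiles, built on FortTeamSquadDetailedEntry.
# On squad assignment, it async-loads icon and ribbon textures into the images' dynamic materials, tinted with the team's text color.
# When the squad-empty state changes, EntryLockSwitcher swaps between members and a lock-reason panel.
# Empty squads also get desaturated, inactive-filled logo and ribbon materials and a darkened name.
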WidgetBlueprint /FortniteCoreUI/Frontend/Parties/SquadManagement/Squad
graphs [EventGraph {nodes [
   n1 = Event OnSquadIsEmptyChanged!() -> (OutputDelegate, then, bSquadIsEmpty:bool, bMultipleSquadsPopulated:bool) {is=title K2Node_Event}
   > n2 = Branch(Condition=@n1.bSquadIsEmpty)
   n3 = Get EntryLockSwitcher() -> (EntryLockSwitcher:Widget Switcher Object Reference)
   > SetActiveWidgetIndex(self=@n3, Index=1)
   n4 = Get SquadLogo() -> (SquadLogo:Image Object Reference)
   > n5 = GetDynamicMaterial(self=@n4)
   > SetScalarParameterValue(self=@n5, ParameterName=DesatFactor, Value=0.3)
   > SetScalarParameterValue(self=@n5, ParameterName=InactiveFill, Value=0.7)
   n6 = Get SquadRibbon() -> (SquadRibbon:Image Object Reference)
   > n7 = GetDynamicMaterial(self=@n6)
   > SetScalarParameterValue(self=@n7, ParameterName=DesatFactor, Value=0.3)
   > SetScalarParameterValue(self=@n7, ParameterName=InactiveFill, Value=0.7)
   n8 = Get DarkenBorder() -> (DarkenBorder:Border Object Reference)
   > SetContentColorAndOpacity!(execute, self:Border Object Reference=@n8, InContentColorAndOpacity:Linear Color Structure=(R=0.2122,G=0.2623,B=0.4452,A=1)) -> (then)
   SetActiveWidgetIndex(execute=@n2.else, self=@n3, Index=0)
   > n9 = GetDynamicMaterial(self=@n4)
   > SetScalarParameterValue(self=@n9, ParameterName=DesatFactor, Value=0)
   > SetScalarParameterValue(self=@n9, ParameterName=InactiveFill, Value=0)
   > n10 = GetDynamicMaterial(self=@n6)
   > SetScalarParameterValue(self=@n10, ParameterName=DesatFactor, Value=0)
   > SetScalarParameterValue(self=@n10, ParameterName=InactiveFill, Value=0)
   > SetContentColorAndOpacity(self=@n8, InContentColorAndOpacity=(R=1,G=1,B=1,A=1))
   n11 = Event OnSquadAssigned!() -> (OutputDelegate, then, SquadStyle:"Athena Team Display Info Structure (by ref)") {is=title K2Node_Event}
   n12 = Break Athena Team Display Info!(AthenaTeamDisplayInfo:"Athena Team Display Info Structure (by ref)"=@n11.SquadStyle) -> (Icon:Soft Object Reference, Ribbon:Soft Object Reference, Name:Text, TextColor:Linear Color Structure, PedestalColor:Linear Color Structure) {is=title K2Node_BreakStruct}
   > n13 = Async Load Asset(Asset=@n12)
   > n14 = Async Load Asset(Asset=@n12.Ribbon)
   n15 = Cast To Texture!(execute=@n14.Completed, Object=@n14) -> (then, CastFailed, AsTexture:Texture Object Reference) {is=title K2Node_DynamicCast}
   > n16 = GetDynamicMaterial(self=Get SquadRibbon())
   > SetTextureParameterValue(self=@n16, ParameterName=RibbonTexture, Value=@n15)
   > SetVectorParameterValue(self=@n16, ParameterName=SquadRGB_2, Value=@n12.TextColor)
   n17 = Cast To Texture(execute=@n13.Completed, Object=@n13)
   > n18 = GetDynamicMaterial(self=Get SquadLogo())
   > SetTextureParameterValue(self=@n18, ParameterName=LogoTexture, Value=@n17)
   > SetVectorParameterValue(self=@n18, ParameterName=SquadRGB_2, Value=@n12.TextColor)
  ]}]
variables []
widgets {
 info {inheritedWidgetCount=3, parentClass=/Script/FortniteUI.FortTeamSquadDetailedEntry, rootWidgetClass=&HorizontalBox, widgetCount=15}
 widgets [
  HorizontalBox_0 {widgetClassPath=&HorizontalBox, widget=HorizontalBox_0}
  VerticalBox_0 {widgetClassPath=&VerticalBox, parent=HorizontalBox_0, widget=VerticalBox_0, slot {class=HorizontalBoxSlot}}
  HorizontalBox_3 {
   widgetClassPath=&HorizontalBox
   parent=VerticalBox_0
   widget=HorizontalBox_3
   properties {flowDirectionPreference=Culture}
   slot {class=VerticalBoxSlot, properties {padding=<ltrb 0 0 0 16>}}
  }
  SquadLogo {
   widgetClassPath=&Image
   bIsVariable=true
   parent=HorizontalBox_3
   widget=SquadLogo
   properties {
    visibility=SelfHitTestInvisible
    brush {
     bIsDynamicallyLoaded=false
     drawAs=Image
     imageType=NoImage
     mirroring=NoMirror
     resourceName=None
     resourceObject=/FortniteCoreUI/UI/SocialPanel/Art/Materials/M_UI_SquadColors_Logo_Master
     tiling=NoTile
     imageSize=<112 112>
     margin=<ltrb 0 0 0 0>
     outlineSettings {
      bUseBrushTransparency=false
      roundingType=HalfHeightRadius
      width=0
      color {colorUseRule=UseColor_Specified, specifiedColor=<rgba 0 0 0 0>}
      cornerRadii=<0 0 0 0>
     }
     tintColor {colorUseRule=UseColor_Specified, specifiedColor=<rgba 1 1 1 1>}
     uVRegion {bIsValid=false, max=<0 0>, min=<0 0>}
    }
   }
   slot {class=HorizontalBoxSlot, properties {verticalAlignment=VAlign_Center, padding=<ltrb 0 0 -32 0>}}
  }
  VerticalBox_3 {
   widgetClassPath=&VerticalBox
   parent=HorizontalBox_3
   widget=VerticalBox_3
   slot {
    class=HorizontalBoxSlot
    properties {horizontalAlignment=HAlign_Center, verticalAlignment=VAlign_Bottom, padding=<ltrb 0 0 0 16>, size {sizeRule=Fill, value=1}}
   }
  }
  SquadRibbon {
   widgetClassPath=&Image
   bIsVariable=true
   parent=VerticalBox_3
   widget=SquadRibbon
   properties {
    visibility=SelfHitTestInvisible
    brush {
     bIsDynamicallyLoaded=false
     drawAs=Image
     imageType=NoImage
     mirroring=NoMirror
     resourceName=None
     resourceObject=/FortniteCoreUI/UI/SocialPanel/Art/Materials/M_UI_SquadColors_Ribbon_Master
     tiling=NoTile
     imageSize=<96 8>
     margin=<ltrb 0 0 0 0>
     outlineSettings {
      bUseBrushTransparency=false
      roundingType=HalfHeightRadius
      width=0
      color {colorUseRule=UseColor_Specified, specifiedColor=<rgba 0 0 0 0>}
      cornerRadii=<0 0 0 0>
     }
     tintColor {colorUseRule=UseColor_Specified, specifiedColor=<rgba 1 1 1 1>}
     uVRegion {bIsValid=false, max=<0 0>, min=<0 0>}
    }
   }
   slot {class=VerticalBoxSlot, properties {horizontalAlignment=HAlign_Center}}
  }
  DarkenBorder {
   widgetClassPath=&Border
   bIsVariable=true
   parent=VerticalBox_3
   widget=DarkenBorder
   properties {brushColor=<rgba 1 1 1 0>}
   slot {class=VerticalBoxSlot, properties {horizontalAlignment=HAlign_Center}}
  }
  RichText_SquadName {
   widgetClassPath=&CommonRichTextBlock
   bInherited=true
   bIsVariable=true
   parent=DarkenBorder
   widget=RichText_SquadName
   properties {
    bOverrideDefaultStyle=true
    justification=Center
    lineHeightPercentage=0.875
    text="Bear <text color=\"FFF\" size=\"18\">Squad</>"
    textTransformPolicy=ToUpper
    defaultTextStyleOverride {
     overflowPolicy=Clip
     transformPolicy=None
     colorAndOpacity {colorUseRule=UseColor_Specified, specifiedColor=<rgba 1 0.1195 0.4072 1>}
     font {
      bForceMonospaced=false
      bMaterialIsStencil=false
      fontMaterial=null
      fontObject=&BurbankSmall
      letterSpacing=0
      monospacedWidth=1
      size=28
      skewAmount=0
      typefaceFontName=Black
      outlineSettings {
       bApplyOutlineToDropShadows=false
       bMiteredCorners=false
       bSeparateFillAlpha=false
       outlineMaterial=null
       outlineSize=0
       outlineColor=<rgba 0 0 0 1>
      }
     }
     highlightColor {colorUseRule=UseColor_Specified, specifiedColor=<rgba 0 0 0 1>}
     highlightShape {
      bIsDynamicallyLoaded=false
      drawAs=Image
      imageType=NoImage
      mirroring=NoMirror
      resourceName=None
      resourceObject=null
      tiling=NoTile
      imageSize=<32 32>
      margin=<ltrb 0 0 0 0>
      outlineSettings {
       bUseBrushTransparency=false
       roundingType=HalfHeightRadius
       width=0
       color {colorUseRule=UseColor_Specified, specifiedColor=<rgba 0 0 0 0>}
       cornerRadii=<0 0 0 0>
      }
      tintColor {colorUseRule=UseColor_Specified, specifiedColor=<rgba 1 1 1 1>}
      uVRegion {bIsValid=false, max=<0 0>, min=<0 0>}
     }
     selectedBackgroundColor {colorUseRule=UseColor_Foreground, specifiedColor=<rgba 1 0 1 1>}
     shadowColorAndOpacity=<rgba 0 0 0 1>
     shadowOffset=<0 0>
     strikeBrush {
      bIsDynamicallyLoaded=false
      drawAs=Image
      imageType=NoImage
      mirroring=NoMirror
      resourceName=None
      resourceObject=null
      tiling=NoTile
      imageSize=<32 32>
      margin=<ltrb 0 0 0 0>
      outlineSettings {
       bUseBrushTransparency=false
       roundingType=HalfHeightRadius
       width=0
       color {colorUseRule=UseColor_Specified, specifiedColor=<rgba 0 0 0 0>}
       cornerRadii=<0 0 0 0>
      }
      tintColor {colorUseRule=UseColor_Specified, specifiedColor=<rgba 1 1 1 1>}
      uVRegion {bIsValid=false, max=<0 0>, min=<0 0>}
     }
     underlineBrush {
      bIsDynamicallyLoaded=false
      drawAs=Image
      imageType=NoImage
      mirroring=NoMirror
      resourceName=None
      resourceObject=null
      tiling=NoTile
      imageSize=<32 32>
      margin=<ltrb 0 0 0 0>
      outlineSettings {
       bUseBrushTransparency=false
       roundingType=HalfHeightRadius
       width=0
       color {colorUseRule=UseColor_Specified, specifiedColor=<rgba 0 0 0 0>}
       cornerRadii=<0 0 0 0>
      }
      tintColor {colorUseRule=UseColor_Specified, specifiedColor=<rgba 1 1 1 1>}
      uVRegion {bIsValid=false, max=<0 0>, min=<0 0>}
     }
    }
    renderTransform {angle=0, scale=<1 1>, shear=<-8 0>, translation=<0 0>}
   }
   slot {class=BorderSlot}
  }
  EntryLockSwitcher {
   widgetClassPath=&WidgetSwitcher
   bIsVariable=true
   parent=VerticalBox_0
   widget=EntryLockSwitcher
   slot {class=VerticalBoxSlot, properties {horizontalAlignment=HAlign_Center, size {sizeRule=Fill, value=1}}}
  }
  EntryBox_Members {
   widgetClassPath=/Script/FortniteUI.FortDynamicEntryBox
   bInherited=true
   bIsVariable=true
   parent=EntryLockSwitcher
   widget=EntryBox_Members
   properties {
    entryBoxType=Vertical
    entryHorizontalAlignment=HAlign_Center
    entryWidgetClass=&/SquadMemberTile.SquadMemberTile_C
    numDesignerPreviewEntries=4
    entrySpacing=<0 8>
   }
   slot {class=WidgetSwitcherSlot}
  }
  LockedReason {
   widgetClassPath=&Border
   parent=EntryLockSwitcher
   widget=LockedReason
   properties {
    flowDirectionPreference=Culture
    background {
     bIsDynamicallyLoaded=false
     drawAs=Image
     imageType=NoImage
     mirroring=NoMirror
     resourceName=None
     resourceObject=/Game/UI/SocialPanel/Art/Materials/M_UI_SocialUserItemBorder_InactiveUserListItem
     tiling=NoTile
     imageSize=<32 32>
     margin=<ltrb 0 0 0 0>
     outlineSettings {
      bUseBrushTransparency=false
      roundingType=HalfHeightRadius
      width=0
      color {colorUseRule=UseColor_Specified, specifiedColor=<rgba 0 0 0 0>}
      cornerRadii=<0 0 0 0>
     }
     tintColor {colorUseRule=UseColor_Specified, specifiedColor=<rgba 1 1 1 1>}
     uVRegion {bIsValid=false, max=<0 0>, min=<0 0>}
    }
    padding=<ltrb 0 0 0 0>
   }
   slot {class=WidgetSwitcherSlot}
  }
  SizeBox_0 {
   widgetClassPath=&SizeBox
   parent=LockedReason
   widget=SizeBox_0
   properties {bOverride_WidthOverride=true, widthOverride=400}
   slot {class=BorderSlot, properties {padding=<ltrb 0 0 0 0>}}
  }
  VerticalBox_1 {widgetClassPath=&VerticalBox, parent=SizeBox_0, widget=VerticalBox_1, slot {class=SizeBoxSlot}}
  Lock {
   widgetClassPath=&Image
   parent=VerticalBox_1
   widget=Lock
   properties {
    bFlipForRightToLeftFlowDirection=true
    visibility=SelfHitTestInvisible
    brush {
     bIsDynamicallyLoaded=false
     drawAs=Image
     imageType=NoImage
     mirroring=NoMirror
     resourceName=None
     resourceObject=/Game/UI/Foundation/Textures/Icons/Locks/T-Icon-Lock-64
     tiling=NoTile
     imageSize=<48 48>
     margin=<ltrb 0 0 0 0>
     outlineSettings {
      bUseBrushTransparency=false
      roundingType=HalfHeightRadius
      width=0
      color {colorUseRule=UseColor_Specified, specifiedColor=<rgba 0 0 0 0>}
      cornerRadii=<0 0 0 0>
     }
     tintColor {colorUseRule=UseColor_Specified, specifiedColor=<rgba 1 1 1 1>}
     uVRegion {bIsValid=false, max=<0 0>, min=<0 0>}
    }
    colorAndOpacity=<rgba 0.0561 0.2346 0.5149 1>
    renderTransform {angle=-5, scale=<1 1>, shear=<0 0>, translation=<0 0>}
   }
   slot {class=VerticalBoxSlot, properties {horizontalAlignment=HAlign_Right, verticalAlignment=VAlign_Top, padding=<ltrb 8 8 8 8>}}
  }
  RichText_LockReason {
   widgetClassPath=&CommonRichTextBlock
   bInherited=true
   bIsVariable=true
   parent=VerticalBox_1
   widget=RichText_LockReason
   properties {
    autoWrapText=true
    bOverrideDefaultStyle=true
    justification=Center
    text="This squad requires <text size=\"32\" color=\"FFF\">99</> <text size=\"18\" color=\"8ABBE6FF\">party members</>"
    textTransformPolicy=ToUpper
    defaultTextStyleOverride {
     overflowPolicy=Clip
     transformPolicy=None
     colorAndOpacity {colorUseRule=UseColor_Specified, specifiedColor=<rgba 0.0561 0.2346 0.5149 1>}
     font {
      bForceMonospaced=false
      bMaterialIsStencil=false
      fontMaterial=null
      fontObject=&BurbankSmall
      letterSpacing=0
      monospacedWidth=1
      size=15
      skewAmount=0
      typefaceFontName=Black
      outlineSettings {
       bApplyOutlineToDropShadows=false
       bMiteredCorners=false
       bSeparateFillAlpha=false
       outlineMaterial=null
       outlineSize=0
       outlineColor=<rgba 0 0 0 1>
      }
     }
     highlightColor {colorUseRule=UseColor_Specified, specifiedColor=<rgba 0 0 0 1>}
     highlightShape {
      bIsDynamicallyLoaded=false
      drawAs=Image
      imageType=NoImage
      mirroring=NoMirror
      resourceName=None
      resourceObject=null
      tiling=NoTile
      imageSize=<32 32>
      margin=<ltrb 0 0 0 0>
      outlineSettings {
       bUseBrushTransparency=false
       roundingType=HalfHeightRadius
       width=0
       color {colorUseRule=UseColor_Specified, specifiedColor=<rgba 0 0 0 0>}
       cornerRadii=<0 0 0 0>
      }
      tintColor {colorUseRule=UseColor_Specified, specifiedColor=<rgba 1 1 1 1>}
      uVRegion {bIsValid=false, max=<0 0>, min=<0 0>}
     }
     selectedBackgroundColor {colorUseRule=UseColor_Foreground, specifiedColor=<rgba 1 0 1 1>}
     shadowColorAndOpacity=<rgba 0 0 0 1>
     shadowOffset=<0 0>
     strikeBrush {
      bIsDynamicallyLoaded=false
      drawAs=Image
      imageType=NoImage
      mirroring=NoMirror
      resourceName=None
      resourceObject=null
      tiling=NoTile
      imageSize=<32 32>
      margin=<ltrb 0 0 0 0>
      outlineSettings {
       bUseBrushTransparency=false
       roundingType=HalfHeightRadius
       width=0
       color {colorUseRule=UseColor_Specified, specifiedColor=<rgba 0 0 0 0>}
       cornerRadii=<0 0 0 0>
      }
      tintColor {colorUseRule=UseColor_Specified, specifiedColor=<rgba 1 1 1 1>}
      uVRegion {bIsValid=false, max=<0 0>, min=<0 0>}
     }
     underlineBrush {
      bIsDynamicallyLoaded=false
      drawAs=Image
      imageType=NoImage
      mirroring=NoMirror
      resourceName=None
      resourceObject=null
      tiling=NoTile
      imageSize=<32 32>
      margin=<ltrb 0 0 0 0>
      outlineSettings {
       bUseBrushTransparency=false
       roundingType=HalfHeightRadius
       width=0
       color {colorUseRule=UseColor_Specified, specifiedColor=<rgba 0 0 0 0>}
       cornerRadii=<0 0 0 0>
      }
      tintColor {colorUseRule=UseColor_Specified, specifiedColor=<rgba 1 1 1 1>}
      uVRegion {bIsValid=false, max=<0 0>, min=<0 0>}
     }
    }
    renderTransform {angle=0, scale=<1 1>, shear=<-8 0>, translation=<0 0>}
   }
   slot {class=VerticalBoxSlot, properties {verticalAlignment=VAlign_Center, padding=<ltrb 16 0 16 32>, size {sizeRule=Fill, value=1}}}
  }
 ]
}
