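# Squad formation menu widget, based on FortSquadSelectionMenu, for moving or swapping squad positions and sitting out or jumping in.
# The layout stacks a squad entry box of Squad widgets, headings, an action prompt, and a switcher holding either Confirm/Cancel or the participation button.
# The only graph logic handles OnLocalPlayerParticipationChanged, setting that button to "Sit Out" in standard style when participating, or "Jump In" in yellow otherwise.
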WidgetBlueprint /FortniteCoreUI/Frontend/Parties/SquadManagement/SquadSelection
graphs [EventGraph {nodes [
   n1 = Event OnLocalPlayerParticipationChanged!() -> (OutputDelegate, then, bIsParticipating:bool) {is=title K2Node_Event}
   > n2 = Cast To IconTextButton!(execute, Object=Get Button_ChangeParticipation() -> (Button_ChangeParticipation:Common Button Legacy Object Reference)) -> (then, CastFailed, AsIcon Text Button:Icon Text Button Object Reference) {is=title K2Node_DynamicCast}
   > Set Text!(
    execute
    self:Icon Text Button Object Reference=@n2
    ButtonText:Text=Select(Option 0:Text="LOCTEXT(\"Jump In\")", Option 1:Text="LOCTEXT(\"Sit Out\")", Index=@n1.bIsParticipating) -> (ReturnValue:Text)
   ) -> (then) {is=title K2Node_CallFunction}
   > SetStyle!(
    execute
    self:Common Button Base Object Reference=@n2
    InStyle:Common Button Style Class Reference=Select(
     Option 0:Common Button Style Class Reference=/FortUILibrary/Foundation/Buttons/ButtonStyles/ButtonStyle_Feature_M_Yellow.ButtonStyle_Feature_M_Yellow_C
     Option 1:Common Button Style Class Reference=/FortUILibrary/Foundation/Buttons/ButtonStyles/ButtonStyle_Feature_M.ButtonStyle_Feature_M_C
     Index=@n1.bIsParticipating
    ) -> (ReturnValue:Common Button Style Class Reference)
   ) -> (then)
  ]}]
variables []
widgets {
 info {inheritedWidgetCount=6, parentClass=/Script/FortniteUI.FortSquadSelectionMenu, rootWidgetClass=&Overlay, widgetCount=15}
 widgets [
  Overlay_0 {widgetClassPath=&Overlay, widget=Overlay_0}
  Border_0 {
   widgetClassPath=&Border
   parent=Overlay_0
   widget=Border_0
   properties {verticalAlignment=VAlign_Center, brushColor=<rgba 0 0.004 0.0331 0.8>, padding=<ltrb 0 0 0 0>}
   slot {class=OverlaySlot, properties {horizontalAlignment=HAlign_Fill, verticalAlignment=VAlign_Fill}}
  }
  VerticalBox_0 {
   widgetClassPath=&VerticalBox
   parent=Border_0
   widget=VerticalBox_0
   slot {class=BorderSlot, properties {verticalAlignment=VAlign_Center, padding=<ltrb 0 0 0 0>}}
  }
  VerticalBox_401 {widgetClassPath=&VerticalBox, parent=VerticalBox_0, widget=VerticalBox_401, slot {class=VerticalBoxSlot}}
  TeamEntryBox_Squads {
   widgetClassPath=/Script/FortniteUI.AthenaTeamSquadsEntryBox
   bInherited=true
   bIsVariable=true
   parent=VerticalBox_401
   widget=TeamEntryBox_Squads
   properties {squadEntryClass=&/Squad.Squad_C, entrySpacing=<40 40>}
   slot {class=VerticalBoxSlot, properties {horizontalAlignment=HAlign_Center}}
  }
  VerticalBox_1 {
   widgetClassPath=&VerticalBox
   parent=VerticalBox_401
   widget=VerticalBox_1
   slot {class=VerticalBoxSlot, properties {horizontalAlignment=HAlign_Center, padding=<ltrb 0 16 0 16>}}
  }
  CommonTextBlock_74 {
   widgetClassPath=&CommonTextBlock
   parent=VerticalBox_1
   widget=CommonTextBlock_74
   properties {
    justification=Center
    text=Squad Formation
    textTransformPolicy=ToUpper
    font {
     bForceMonospaced=false
     bMaterialIsStencil=false
     fontMaterial=null
     fontObject=&Burbank
     letterSpacing=0
     monospacedWidth=1
     size=32
     skewAmount=0
     typefaceFontName=BurbankRegular-Black
     outlineSettings {
      bApplyOutlineToDropShadows=false
      bMiteredCorners=false
      bSeparateFillAlpha=false
      outlineMaterial=null
      outlineSize=0
      outlineColor=<rgba 0 0 0 1>
     }
    }
    renderTransform {angle=0, scale=<1 1>, shear=<-8 0>, translation=<0 0>}
    shadowOffset=<0 0>
   }
   slot {class=VerticalBoxSlot}
  }
  CommonTextBlock_163 {
   widgetClassPath=&CommonTextBlock
   parent=VerticalBox_1
   widget=CommonTextBlock_163
   properties {
    justification=Center
    text=Move or Swap Positions
    textTransformPolicy=ToUpper
    colorAndOpacity {colorUseRule=UseColor_Specified, specifiedColor=<rgba 0.0561 0.2346 0.5149 1>}
    font {
     bForceMonospaced=false
     bMaterialIsStencil=false
     fontMaterial=null
     fontObject=&BurbankSmall
     letterSpacing=0
     monospacedWidth=1
     size=18
     skewAmount=0
     typefaceFontName=Black
     outlineSettings {
      bApplyOutlineToDropShadows=false
      bMiteredCorners=false
      bSeparateFillAlpha=false
      outlineMaterial=null
      outlineSize=0
      outlineColor=<rgba 0 0 0 1>
     }
    }
    renderTransform {angle=0, scale=<1 1>, shear=<-8 0>, translation=<0 0>}
    shadowOffset=<0 0>
   }
   slot {class=VerticalBoxSlot}
  }
  SizeBox_0 {
   widgetClassPath=&SizeBox
   parent=VerticalBox_401
   widget=SizeBox_0
   properties {bOverride_HeightOverride=true, heightOverride=64}
   slot {class=VerticalBoxSlot, properties {padding=<ltrb 0 32 0 16>}}
  }
  RichText_ActionPrompt {
   widgetClassPath=&CommonRichTextBlock
   bInherited=true
   bIsVariable=true
   parent=SizeBox_0
   widget=RichText_ActionPrompt
   properties {
    bOverrideDefaultStyle=true
    justification=Center
    text="Basic message styling <text case=\"upper\" size=\"18\" fontface=\"black\">basic callout styling</> <text case=\"upper\" color=\"ff61abff\" size=\"18\" fontface=\"black\">squad name styling</> <text case=\"upper\" color=\"FFF\" size=\"18\" fontface=\"black\">squad styling</> <text case=\"upper\" color=\"b2ffff\" size=\"18\" fontface=\"black\">user name styling</>"
    wrapTextAt=1280
    defaultTextStyleOverride {
     overflowPolicy=Clip
     transformPolicy=None
     colorAndOpacity {colorUseRule=UseColor_Specified, specifiedColor=<rgba 0.2542 0.4969 0.7913 1>}
     font {
      bForceMonospaced=false
      bMaterialIsStencil=false
      fontMaterial=null
      fontObject=&BurbankSmall
      letterSpacing=0
      monospacedWidth=1
      size=15
      skewAmount=0
      typefaceFontName=Bold
      outlineSettings {
       bApplyOutlineToDropShadows=false
       bMiteredCorners=false
       bSeparateFillAlpha=false
       outlineMaterial=null
       outlineSize=0
       outlineColor=<rgba 0 0 0 1>
      }
     }
     highlightColor {colorUseRule=UseColor_Specified, specifiedColor=<rgba 0 0 0 1>}
     highlightShape {
      bIsDynamicallyLoaded=false
      drawAs=Image
      imageType=NoImage
      mirroring=NoMirror
      resourceName=None
      resourceObject=null
      tiling=NoTile
      imageSize=<32 32>
      margin=<ltrb 0 0 0 0>
      outlineSettings {
       bUseBrushTransparency=false
       roundingType=HalfHeightRadius
       width=0
       color {colorUseRule=UseColor_Specified, specifiedColor=<rgba 0 0 0 0>}
       cornerRadii=<0 0 0 0>
      }
      tintColor {colorUseRule=UseColor_Specified, specifiedColor=<rgba 1 1 1 1>}
      uVRegion {bIsValid=false, max=<0 0>, min=<0 0>}
     }
     selectedBackgroundColor {colorUseRule=UseColor_Foreground, specifiedColor=<rgba 1 0 1 1>}
     shadowColorAndOpacity=<rgba 0 0 0 1>
     shadowOffset=<0 0>
     strikeBrush {
      bIsDynamicallyLoaded=false
      drawAs=Image
      imageType=NoImage
      mirroring=NoMirror
      resourceName=None
      resourceObject=null
      tiling=NoTile
      imageSize=<32 32>
      margin=<ltrb 0 0 0 0>
      outlineSettings {
       bUseBrushTransparency=false
       roundingType=HalfHeightRadius
       width=0
       color {colorUseRule=UseColor_Specified, specifiedColor=<rgba 0 0 0 0>}
       cornerRadii=<0 0 0 0>
      }
      tintColor {colorUseRule=UseColor_Specified, specifiedColor=<rgba 1 1 1 1>}
      uVRegion {bIsValid=false, max=<0 0>, min=<0 0>}
     }
     underlineBrush {
      bIsDynamicallyLoaded=false
      drawAs=Image
      imageType=NoImage
      mirroring=NoMirror
      resourceName=None
      resourceObject=null
      tiling=NoTile
      imageSize=<32 32>
      margin=<ltrb 0 0 0 0>
      outlineSettings {
       bUseBrushTransparency=false
       roundingType=HalfHeightRadius
       width=0
       color {colorUseRule=UseColor_Specified, specifiedColor=<rgba 0 0 0 0>}
       cornerRadii=<0 0 0 0>
      }
      tintColor {colorUseRule=UseColor_Specified, specifiedColor=<rgba 1 1 1 1>}
      uVRegion {bIsValid=false, max=<0 0>, min=<0 0>}
     }
    }
   }
   slot {class=SizeBoxSlot, properties {verticalAlignment=VAlign_Top}}
  }
  Switcher_Buttons {
   widgetClassPath=&WidgetSwitcher
   bInherited=true
   bIsVariable=true
   parent=VerticalBox_401
   widget=Switcher_Buttons
   slot {class=VerticalBoxSlot, properties {horizontalAlignment=HAlign_Center}}
  }
  ConfirmButtons {
   widgetClassPath=&HorizontalBox
   parent=Switcher_Buttons
   widget=ConfirmButtons
   properties {flowDirectionPreference=Culture}
   slot {class=WidgetSwitcherSlot}
  }
  Button_ConfirmMove {
   widgetClassPath=&IconTextButton
   bInherited=true
   bIsVariable=true
   parent=ConfirmButtons
   widget=Button_ConfirmMove
   properties {
    bIsFocusable=false
    buttonText=Confirm
    minWidth=256
    onButtonBaseClicked=(&self:WidgetTree.Button_ConfirmMove.HandleOnButtonClicked)
    onButtonBaseDoubleClicked=(&self:WidgetTree.Button_ConfirmMove.HandleOnButtonDoubleClicked)
    onButtonBaseHovered=(&self:WidgetTree.Button_ConfirmMove.HandleOnButtonHovered)
    onButtonBaseUnhovered=(&self:WidgetTree.Button_ConfirmMove.HandleOnButtonUnhovered)
    onSelectedChangedBase=(&self:WidgetTree.Button_ConfirmMove.HandleOnSelectedChanged)
    style=&ButtonStyle_Feature_M_Yellow_C
    triggeringInputAction {dataTable=&InputActionDataTable, rowName=Join}
   }
   slot {class=HorizontalBoxSlot, properties {padding=<ltrb 0 0 32 0>}}
  }
  Button_CancelMove {
   widgetClassPath=&IconTextButton
   bInherited=true
   bIsVariable=true
   parent=ConfirmButtons
   widget=Button_CancelMove
   properties {
    bIsFocusable=false
    buttonText=Cancel
    minWidth=256
    onButtonBaseClicked=(&self:WidgetTree.Button_CancelMove.HandleOnButtonClicked)
    onButtonBaseDoubleClicked=(&self:WidgetTree.Button_CancelMove.HandleOnButtonDoubleClicked)
    onButtonBaseHovered=(&self:WidgetTree.Button_CancelMove.HandleOnButtonHovered)
    onButtonBaseUnhovered=(&self:WidgetTree.Button_CancelMove.HandleOnButtonUnhovered)
    onSelectedChangedBase=(&self:WidgetTree.Button_CancelMove.HandleOnSelectedChanged)
    style=&ButtonStyle_Feature_M_C
    triggeringInputAction {dataTable=&InputActionDataTable, rowName=GenericBack}
   }
   slot {class=HorizontalBoxSlot}
  }
  Button_ChangeParticipation {
   widgetClassPath=&IconTextButton
   bInherited=true
   bIsVariable=true
   parent=Switcher_Buttons
   widget=Button_ChangeParticipation
   properties {
    bIsFocusable=false
    buttonText=Sit Out
    onButtonBaseClicked=(&self:WidgetTree.Button_ChangeParticipation.HandleOnButtonClicked)
    onButtonBaseDoubleClicked=(&self:WidgetTree.Button_ChangeParticipation.HandleOnButtonDoubleClicked)
    onButtonBaseHovered=(&self:WidgetTree.Button_ChangeParticipation.HandleOnButtonHovered)
    onButtonBaseUnhovered=(&self:WidgetTree.Button_ChangeParticipation.HandleOnButtonUnhovered)
    onSelectedChangedBase=(&self:WidgetTree.Button_ChangeParticipation.HandleOnSelectedChanged)
    style=&ButtonStyle_Feature_M_C
    triggeringInputAction {dataTable=&InputActionDataTable, rowName=Join}
   }
   slot {class=WidgetSwitcherSlot}
  }
 ]
}
